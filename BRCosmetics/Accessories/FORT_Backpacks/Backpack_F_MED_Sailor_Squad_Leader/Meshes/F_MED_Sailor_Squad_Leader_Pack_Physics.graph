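# Physics asset for the Sailor Squad Leader back bling, simulating its dangling accessories such as pistol, bunny plush, banana, necklace, buttons, stars and strap.
# Box and capsule bodies on the dyn_ bones are constrained to the pack body, with chains for pistol to strap, bunny head to ears, and the banana segments.
# The pistol and backpack strap swing and twist freely, the middle banana joint is locked, and most other joints have angular limits.
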PhysicsAsset /BRCosmetics/Accessories/FORT_Backpacks/Backpack_F_MED_Sailor_Squad_Leader/Meshes/F_MED_Sailor_Squad_Leader_Pack_Physics
bodies {
 dyn_backpack_strap_MSTR [{shapeType=Capsule, length=4.4705, radius=2.4686, center=<-1.1091 -1.1099 -7.4304>, rotation=<rot 19.2907 -84.617 15.9184>}]
 dyn_banana [{shapeType=Capsule, length=8.3398, radius=4.4807, center=<0.1135 0.4 4>, rotation=<rot 0 -160.7644 -52.5819>}]
 dyn_banana_01 [{shapeType=Capsule, length=1.185, radius=1.1261, rotation=<rot 0 0 90>}]
 dyn_banana_MSTR [{shapeType=Capsule, length=0.505, radius=0.505}]
 dyn_bunny_head [{shapeType=Capsule, length=0.209, radius=5.7495, center=<1.0863 0 4.1073>}]
 dyn_bunny_l_arm [{shapeType=Capsule, length=3.9366, radius=1.6538, center=<2.7298 0 0>, rotation=<rot -90 -1.8984 1.8984>}]
 dyn_bunny_l_ear_01 [{shapeType=Capsule, length=8.4048, radius=1.4384, center=<6.5081 -0.464 -0.304>, rotation=<rot -84.9968 179.9996 -179.9996>}]
 dyn_bunny_r_arm [{shapeType=Capsule, length=5.6193, radius=1.9574, center=<4.2142 -0.0001 0.3557>, rotation=<rot 0.0005 89.9993 89.9997>}]
 dyn_bunny_r_ear_01 [{shapeType=Capsule, length=3.9055, radius=1.2536, center=<3.1389 0.5281 -0.2633>, rotation=<rot 79.9996 0 -4.9999>}]
 dyn_button_01 [{shapeType=Box, extentX=1.01, extentY=3.5096, extentZ=3.3386, center=<0.3639 -0.0942 0.0972>}]
 dyn_button_02 [{shapeType=Box, extentX=1.01, extentY=3.5096, extentZ=3.3386, center=<0.4441 0.2471 -0.2097>}]
 dyn_necklace [{shapeType=Capsule, length=2.891, radius=1.088, center=<2.6902 -0.9837 -0.7993>, rotation=<rot -1.0299 -21.5645 95.0603>}]
 dyn_pistol_MSTR [
  {shapeType=Box, extentX=7.5, extentY=3.7881, extentZ=5.8116, center=<4.968 -1.5168 -2.7795>, rotation=<rot -14.9925 -1.8272 0.4729>}
  {shapeType=Box, extentX=4.7388, extentY=19.2315, extentZ=13.327, center=<14.6473 11.1694 -9.9593>, rotation=<rot -44.9998 0.0001 0>}
 ]
 dyn_pistol_strap [{shapeType=Box, extentX=2.3808, extentY=2.5897, extentZ=0.5528, center=<2.8594 -1.0407 0>, rotation=<rot 0 -19.9998 0>}]
 dyn_stars [{shapeType=Box, extentX=1.01, extentY=4.2125, extentZ=4.2125, center=<2.2023 0 0>}]
 pack [
  {shapeType=Capsule, length=4.7608, radius=3.6347, center=<0.3036 -3.4601 -8.395>}
  {shapeType=Box, extentX=2.5061, extentY=12.1985, extentZ=16.3967, center=<11.4752 -8.0336 -2.4353>, rotation=<rot 5 0 0>}
  {shapeType=Box, extentX=21.4535, extentY=1.8052, extentZ=7.5, center=<0 -11.4624 8.6659>, rotation=<rot 0 0 29.9987>}
  {shapeType=Box, extentX=2.0765, extentY=13.157, extentZ=14.6312, center=<8.266 -7.6534 6.2303>, rotation=<rot 19.6834 -10.6282 -3.6165>}
 ]
}
constraints [
 {
  bone1Name=dyn_pistol_MSTR
  bone2Name=pack
  swing1LimitDegrees=45
  swing1Motion=Free
  swing2LimitDegrees=45
  swing2Motion=Free
  twistLimitDegrees=45
  twistMotion=Free
 }
 {bone1Name=dyn_button_01, bone2Name=pack, swing1LimitDegrees=2, swing2LimitDegrees=2, twistLimitDegrees=5}
 {bone1Name=dyn_button_02, bone2Name=pack, swing1LimitDegrees=2, swing2LimitDegrees=2, twistLimitDegrees=5}
 {bone1Name=dyn_pistol_strap, bone2Name=dyn_pistol_MSTR, swing1LimitDegrees=45, swing2LimitDegrees=45, twistLimitDegrees=3}
 {bone1Name=dyn_stars, bone2Name=pack, swing1LimitDegrees=3, swing2LimitDegrees=7, twistLimitDegrees=5}
 {bone1Name=dyn_bunny_head, bone2Name=pack, swing1LimitDegrees=5, swing2LimitDegrees=2, twistLimitDegrees=7}
 {bone1Name=dyn_bunny_l_ear_01, bone2Name=dyn_bunny_head, swing1LimitDegrees=15, swing2LimitDegrees=15, twistLimitDegrees=3}
 {bone1Name=dyn_bunny_r_ear_01, bone2Name=dyn_bunny_head, swing1LimitDegrees=10, swing2LimitDegrees=10, twistLimitDegrees=3}
 {bone1Name=dyn_bunny_l_arm, bone2Name=pack, swing1LimitDegrees=45, swing2LimitDegrees=45, twistLimitDegrees=3}
 {bone1Name=dyn_bunny_r_arm, bone2Name=pack, swing1LimitDegrees=15, swing2LimitDegrees=45, twistLimitDegrees=3}
 {bone1Name=dyn_necklace, bone2Name=pack, swing1LimitDegrees=45, swing2LimitDegrees=45, twistLimitDegrees=15}
 {bone1Name=dyn_banana_MSTR, bone2Name=pack, swing1LimitDegrees=3, swing2LimitDegrees=15, twistLimitDegrees=2}
 {
  bone1Name=dyn_banana_01
  bone2Name=dyn_banana_MSTR
  swing1LimitDegrees=0.1
  swing1Motion=Locked
  swing2LimitDegrees=0.8
  swing2Motion=Locked
  twistLimitDegrees=0.5
  twistMotion=Locked
 }
 {bone1Name=dyn_banana, bone2Name=dyn_banana_01, swing1LimitDegrees=2, swing2LimitDegrees=0, swing2Motion=Locked, twistLimitDegrees=4}
 {
  bone1Name=dyn_backpack_strap_MSTR
  bone2Name=pack
  swing1LimitDegrees=45
  swing1Motion=Free
  swing2LimitDegrees=45
  swing2Motion=Free
  twistLimitDegrees=45
  twistMotion=Free
 }
]
