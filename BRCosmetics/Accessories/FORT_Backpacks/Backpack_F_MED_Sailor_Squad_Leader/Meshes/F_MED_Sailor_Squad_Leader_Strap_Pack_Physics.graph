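# Physics asset for the strap on the Sailor Squad Leader cosmetic backpack, giving it a jointed chain for physical motion.
# Capsule bodies cover the pack and three strap bones, chained from pack through the master strap bone to two segments.
# The master-to-pack joint swings and twists freely, while the segment joints set moderate swing and tight twist limits.
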PhysicsAsset /BRCosmetics/Accessories/FORT_Backpacks/Backpack_F_MED_Sailor_Squad_Leader/Meshes/F_MED_Sailor_Squad_Leader_Strap_Pack_Physics
bodies {
 dyn_backpack_strap [{shapeType=Capsule, length=1.5287, radius=0.9984, center=<1.7698 0 0>, rotation=<rot 90 0 0>}]
 dyn_backpack_strap_01 [{shapeType=Capsule, length=1.6109, radius=1.1259, center=<1.6434 0 0>, rotation=<rot 90 0 0>}]
 dyn_backpack_strap_MSTR [{shapeType=Capsule, length=0.505, radius=0.505}]
 pack [{shapeType=Capsule, length=3.2628, radius=2.2139, center=<-0.535 -7.1667 -2.566>, rotation=<rot 90 0 0>}]
}
constraints [
 {
  bone1Name=dyn_backpack_strap_MSTR
  bone2Name=pack
  swing1LimitDegrees=45
  swing1Motion=Free
  swing2LimitDegrees=45
  swing2Motion=Free
  twistLimitDegrees=45
  twistMotion=Free
 }
 {bone1Name=dyn_backpack_strap, bone2Name=dyn_backpack_strap_MSTR, swing1LimitDegrees=30, swing2LimitDegrees=30, twistLimitDegrees=5}
 {bone1Name=dyn_backpack_strap_01, bone2Name=dyn_backpack_strap, swing1LimitDegrees=30, swing2LimitDegrees=30, twistLimitDegrees=5}
]
